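# Backup material for the Dueling Dragons Octane car skin, building an animated near-black and cream pattern.
# On UV1, noise G sets the base blend, while slowly rotating, offset or mirrored noise samples are selected by RGBA masks.
# These animated layers, gated by noise B and A, push toward cream and black respectively.
# A two-sided Fresnel falloff shades the result, which the graph shows connected to no material input.
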Material /VehicleCosmetics/Bodies/Octane/Skins/OctaneDuelingdragons/MAT_Skin_Octane_DuelingDragons_Backup
graph {
 properties {bAutomaticallySetUsageInEditor=false, usedShadingModels=Default Lit}
 nodes [
  Constant2Vector() {r=0.5}
  n1 = Constant3Vector() {constant=<rgba 0.01 0.01 0.01 0>}
  n2 = Constant3Vector() {constant=<rgba 0.774 0.767 0.673 0>}
  n3 = TextureCoordinate() {coordinateIndex=1, desc=positive}
  n4 = TextureSample(UVs=@n3) {texture=&/T_DuelingDragons_Noise}
  n5 = LinearInterpolate(A=@n1, B=@n2, Alpha=@n4.G)
  n6 = TextureCoordinate() {coordinateIndex=1, desc=Negative, uTiling=-1}
  n7 = Multiply(A=Time() {bOverride_Period=true, period=600}, B=ScalarParameter() {defaultValue=0.5, parameterName=Rotator speed})
  n8 = Rotator(Coordinate=Add(A=@n6, B=Constant2Vector() {g=0.04, r=0.2}), Time=@n7) {centerX=-0.5}
  n9 = Rotator(Coordinate=Add(A=@n3, B=Constant2Vector() {g=0.31}), Time=@n7)
  n10 = TextureSample(UVs=@n3) {samplerType=SAMPLERTYPE_Masks, texture=&/T_DuelingDragons_RGB}
  n11 = Add(A=@n6, B=Constant2Vector() {r=0.5})
  n12 = LinearInterpolate(
   A=LinearInterpolate(
    A=TextureSample(UVs=@n8) {texture=&T_DuelingDragons_Noise}.R
    B=TextureSample(UVs=@n9) {texture=&T_DuelingDragons_Noise}.R
    Alpha=@n10.A
   )
   B=TextureSample(UVs=Rotator(Coordinate=@n11, Time=@n7) {centerX=-0.5}) {texture=&T_DuelingDragons_Noise}.R
   Alpha=@n10.R
  )
  n13 = Constant2Vector() {g=-0.3, r=0.45}
  n14 = Add(A=@n3, B=@n13)
  n15 = Add(A=@n6, B=@n13)
  n16 = ConstantBiasScale(None=ComponentMask(None=@n3) {r=true}) {bias=-0.5, scale=10000}
  n17 = TextureSample(UVs=LinearInterpolate(A=Rotator(Coordinate=@n14, Time=@n7), B=Rotator(Coordinate=@n15, Time=@n7) {centerX=-0.5}, Alpha=Saturate(None=@n16))) {texture=&T_DuelingDragons_Noise}
  n18 = Rotator(Coordinate=Add(A=@n3, B=Constant2Vector() {g=-0.3, r=0.15}), Time=@n7)
  n19 = Multiply(
   A=@n4.B
   B=LinearInterpolate(
    A=LinearInterpolate(A=@n12, B=@n17.R, Alpha=@n10.G)
    B=TextureSample(UVs=@n18) {texture=&T_DuelingDragons_Noise}.R
    Alpha=@n10.B
   )
  )
  n20 = Constant2Vector() {g=0.35}
  n21 = TextureSample(UVs=Rotator(Coordinate=Add(A=@n6, B=@n20), Time=@n7) {centerX=-0.5}) {texture=&T_DuelingDragons_Noise}
  n22 = Constant2Vector() {g=-0.025, r=0.1}
  n23 = TextureSample(UVs=Rotator(Coordinate=Add(A=@n3, B=@n22), Time=@n7)) {texture=&T_DuelingDragons_Noise}
  n24 = Rotator(Coordinate=Add(A=@n3, B=Constant2Vector() {r=-0.5}), Time=@n7) {centerX=-0.5}
  n25 = Add(A=@n6, B=Constant2Vector() {g=-0.25, r=0.3})
  n26 = LinearInterpolate(
   A=LinearInterpolate(
    A=LinearInterpolate(A=@n21.R, B=@n23.R, Alpha=@n10.A)
    B=TextureSample(UVs=@n24) {texture=&T_DuelingDragons_Noise}.R
    Alpha=@n10.R
   )
   B=TextureSample(UVs=Rotator(Coordinate=@n25, Time=@n7) {centerX=-0.5}) {texture=&T_DuelingDragons_Noise}.R
   Alpha=@n10.B
  )
  n27 = Fresnel() {baseReflectFraction=0, exponent=3}
  n28 = Power(Base=OneMinus(None=@n27), Exp=ScalarParameter() {defaultValue=2, parameterName=FinishInteriorFalloff})
  n29 = Power(Base=@n27, Exp=ScalarParameter() {defaultValue=2, parameterName=FinishOuterFalloff})
  Multiply(
   A=LinearInterpolate(A=LinearInterpolate(A=@n5, B=@n2, Alpha=@n19), B=@n1, Alpha=Multiply(A=@n4.A, B=@n26))
   B=Saturate(None=Max(A=@n28, B=@n29))
  )
 ]
}
